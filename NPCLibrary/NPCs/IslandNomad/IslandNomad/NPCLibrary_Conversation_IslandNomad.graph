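# Conversation database for the Island Nomad NPC, defining its greeting and the branches that follow.
# A single entry point leads to a Speech node with a default line plus contextual lines for a first-ever conversation or specific player character cosmetics, including Island Nomad's own.
# The speech then branches through link nodes into the shared NPC Library quests and basic services conversations, referenced by entry tag.
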ConversationDatabase /NPCLibrary/NPCs/IslandNomad/IslandNomad/NPCLibrary_Conversation_IslandNomad
entry_points [{tag=Conversation.Entry.IslandNomad, destinations [#1]}]
nodes [
 ConversationEntryPointNode_7 {
  id=#1
  class=ConversationEntryPointNode
  outputs [#2]
  properties {entryTag {tagName=Conversation.Entry.IslandNomad}}
 }
 Speech_C_4 {
  id=#2
  class=Speech_C
  outputs [#3, #4]
  properties {
   nodeName=""
   generalConfig {
    defaultMessage=Something's not right.
    contextualMessages [
     {
      message="I followed some klombo tracks here, but they disappear..."
      contextRequirements [{requirement {class=FortControllerRequirement_IsFirstEverConversationWithNPC}}]
      priority {curve {rowName=Tandem.Speech.Priority.FirstEverInteraction}}
     }
     {
      message="Whatever experiments you're doing, run them faster. We're in trouble."
      contextRequirements [{requirement {
         class=FortControllerRequirement_HasCID
         properties {allowedCIds [/BRCosmetics/Athena/Items/Cosmetics/Characters/CID_540_Athena_Commando_M_MeteorManRemix]}
        }}]
      priority {value=0, curve {curveTable=null, rowName=None}}
     }
     {
      message=Have you come to help the Island?
      contextRequirements [{requirement {
         class=FortControllerRequirement_HasCID
         properties {allowedCIds [/BRCosmetics/Athena/Items/Cosmetics/Characters/CID_A_424_Athena_Commando_M_Lancelot]}
        }}]
      priority {value=0, curve {curveTable=null, rowName=None}}
     }
     {
      message=You look familiar... Are you here to help?
      contextRequirements [{requirement {
         class=FortControllerRequirement_HasCID
         properties {allowedCIds [/BRCosmetics/Athena/Items/Cosmetics/Characters/CID_A_428_Athena_Commando_F_PinkWidow]}
        }}]
      priority {value=0, curve {curveTable=null, rowName=None}}
     }
     {
      message="We need a noble warrior, now more than ever."
      contextRequirements [{requirement {
         class=FortControllerRequirement_HasCID
         properties {allowedCIds [/BRCosmetics/Athena/Items/Cosmetics/Characters/CID_A_427_Athena_Commando_F_Fuchsia]}
        }}]
      priority {value=0, curve {curveTable=null, rowName=None}}
     }
     {
      message=Did the Reality Tree do this to you?
      contextRequirements [{requirement {
         class=FortControllerRequirement_HasCID
         properties {allowedCIds [/BRCosmetics/Athena/Items/Cosmetics/Characters/CID_A_429_Athena_Commando_M_Collectable]}
        }}]
      priority {value=0, curve {curveTable=null, rowName=None}}
     }
     {
      message=The Peace Syndicate has the resources to look into what's happening here.
      contextRequirements [{requirement {
         class=FortControllerRequirement_HasCID
         properties {allowedCIds [&CID_A_425_Athena_Commando_F_BlueJay]}
        }}]
      priority {value=0, curve {curveTable=null, rowName=None}}
     }
     {
      message=Have you seen any klombo on your travels?
      contextRequirements [{requirement {
         class=FortControllerRequirement_HasCID
         properties {allowedCIds [/BRCosmetics/Athena/Items/Cosmetics/Characters/CID_A_423_Athena_Commando_M_Canary]}
        }}]
      priority {value=0, curve {curveTable=null, rowName=None}}
     }
     {
      message=Your dark energy is not welcome here.
      contextRequirements [{requirement {
         class=FortControllerRequirement_HasCID
         properties {allowedCIds [&CID_A_422_Athena_Commando_M_Realm]}
        }}]
      priority {value=0, curve {curveTable=null, rowName=None}}
     }
     {
      message=The klombo disappeared in the same strange flash of light that brought them here.
      contextRequirements [{requirement {
         class=FortControllerRequirement_HasCID
         properties {allowedCIds [/BRCosmetics/Athena/Items/Cosmetics/Characters/CID_A_291_Athena_Commando_F_IslandNomad]}
        }}]
      priority {value=0, curve {curveTable=null, rowName=None}}
     }
    ]
   }
  }
 }
 ConversationLinkNode_0 {id=#4, class=ConversationLinkNode, properties {remoteEntryTag {tagName=Conversation.Entry.BasicServices}}}
 ConversationLinkNode_1 {id=#3, class=ConversationLinkNode, properties {remoteEntryTag {tagName=Conversation.Entry.NPCLibrary.Quests}}}
]
